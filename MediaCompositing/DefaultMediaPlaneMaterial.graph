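# Default unlit material for a media plane in the MediaCompositing plugin, built to display a media texture.
# Its graph is a single texture sample fed by the InputTexture object parameter, which defaults to DefaultMediaPlaneTexture and can be overridden by material instances.
# Both parameter and sample use the external sampler type, so the input is treated as an external texture such as a video frame.
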Material /MediaCompositing/DefaultMediaPlaneMaterial
graph {
 properties {pixelDepthOffsetMode=PDOM_Legacy, shadingModel=MSM_Unlit, usedShadingModels=Unlit}
 nodes [TextureSample(Tex=TextureObjectParameter() {parameterName=InputTexture, samplerType=SAMPLERTYPE_External, texture=&/DefaultMediaPlaneTexture}) {samplerType=SAMPLERTYPE_External}]
}
